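# MetaSound Source for the ClayPlug minigun vox, playing the ClydeGrotto minigun fire loop music wave with its own attenuation and the ranged weapon sound class.
# A single stereo Wave Player starts on play, taking the wave, a loop flag and a float start time converted to Time as inputs.
# Besides stereo audio, it outputs the player's Playback Location as NormalizedLyricTime, exposing the position within the wave.
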MetaSoundSource /Clyde/Audio/Items/ClayPlug_Weapons/MSS_ClayPlug_Minigun_Vox
inputs [
 Loops {type=Bool}
 StartTime {type=Float}
 UE.Source.OnPlay {type=Trigger}
 Wave Asset {type=WaveAsset}
]
interfaces [UE.OutputFormat.Stereo:1.0, UE.Source:1.1]
outputs [
 NormalizedLyricTime {type=Float}
 "UE.OutputFormat.Stereo.Audio:0" {type=Audio}
 "UE.OutputFormat.Stereo.Audio:1" {type=Audio}
]
pages [Default {
  input_defaults {Loops=null, StartTime=0, UE.Source.OnPlay=null, Wave Asset=&/MX_IP_ClydeGrotto_MG_Fire_Loop}
  nodes [WavePlayer_0 = UE.Wave Player:Stereo(Loop=^Loops, Play=^UE.Source.OnPlay, Wave Asset=^Wave Asset, Start Time=UE.ConversionFloatToTime(In=^StartTime).Out)]
  links [
   {dst="@UE.OutputFormat.Stereo.Audio:0", src=WavePlayer_0.Out Left}
   {dst="@UE.OutputFormat.Stereo.Audio:1", src=WavePlayer_0.Out Right}
   {dst="@NormalizedLyricTime", src=WavePlayer_0.Playback Location}
  ]
 }]
properties {
 assetImportData=&self:AssetImportData
 attenuationSettings=&/ATT_ClayPlug_Minigun_VOX
 bLooping=true
 duration=10000
 maxDistance=3300
 numChannels=2
 outputFormat=Stereo
 sampleRate=48000
 soundClassObject=&Fort_WeaponRanged
 assetUserData [&self:AudioAssetUserData_0]
}
